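# Physics asset for the left leg of the FeatherMudLounge medium male body cosmetic, set up for secondary motion on the shorts.
# Four dyn_shorts capsules each constrain to their own tiny base_shorts anchor body, with limited swing and zero-degree twist limits.
# The anchors and a calf capsule link to the thigh capsule through angularly free constraints, with thigh and calf providing the leg's volume.
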PhysicsAsset /BRCosmetics/Characters/Player/Male/Medium/Bodies/M_MED_FeatherMudLounge/Meshes/M_MED_FeatherMudLounge_Left_leg_Physics
bodies {
 base_shorts_l_01 [{shapeType=Capsule, length=0.5, radius=0.5, shapeName=dyn_shorts_r_01_capsule}]
 base_shorts_l_02 [{shapeType=Capsule, length=0.5, radius=0.5, shapeName=dyn_shorts_r_03_capsule}]
 base_shorts_l_03 [{shapeType=Capsule, length=0.5, radius=0.5, shapeName=dyn_shorts_r_01_capsule}]
 base_shorts_l_04 [{shapeType=Capsule, length=0.5, radius=0.5, shapeName=dyn_shorts_r_01_capsule}]
 calf_twist_02_l [{
   shapeType=Capsule
   length=8.9682
   radius=4.2677
   shapeName=calf_twist_02_l_capsule
   center=<-4.2447 -11.3589 0.0286>
   rotation=<rot 0.2036 -96.8742 -175.6552>
  }]
 dyn_shorts_l_01 [{shapeType=Capsule, length=6, radius=1.5, shapeName=dyn_shorts_r_02_capsule, center=<4.5 0 0>, rotation=<rot 90 180 0>}]
 dyn_shorts_l_02 [{shapeType=Capsule, length=6, radius=1.5, shapeName=dyn_shorts_r_04_capsule, center=<4.5 0 0>, rotation=<rot 90 180 0>}]
 dyn_shorts_l_03 [{shapeType=Capsule, length=6, radius=1.5, shapeName=dyn_shorts_r_03_capsule, center=<4.5 0 0>, rotation=<rot 90 180 180>}]
 dyn_shorts_l_04 [{shapeType=Capsule, length=6, radius=1.5, shapeName=dyn_shorts_r_01_capsule, center=<4.5 0 0>, rotation=<rot 90 0 0>}]
 thigh_l [{
   shapeType=Capsule
   length=24.1873
   radius=6.5485
   shapeName=thigh_r_capsule
   center=<-21.1553 -0.3181 0.4591>
   rotation=<rot -0.1947 -94.3984 87.4692>
  }]
}
constraints [
 {bone1Name=dyn_shorts_l_01, bone2Name=base_shorts_l_01, swing1LimitDegrees=20, swing2LimitDegrees=30, twistLimitDegrees=0}
 {bone1Name=dyn_shorts_l_02, bone2Name=base_shorts_l_02, swing1LimitDegrees=20, swing2LimitDegrees=15, twistLimitDegrees=0}
 {bone1Name=dyn_shorts_l_03, bone2Name=base_shorts_l_03, swing1LimitDegrees=25, swing2LimitDegrees=18, twistLimitDegrees=0}
 {bone1Name=dyn_shorts_l_04, bone2Name=base_shorts_l_04, swing1LimitDegrees=25, swing2LimitDegrees=30, twistLimitDegrees=0}
 {
  bone1Name=base_shorts_l_01
  bone2Name=thigh_l
  swing1LimitDegrees=45
  swing1Motion=Free
  swing2LimitDegrees=45
  swing2Motion=Free
  twistLimitDegrees=45
  twistMotion=Free
 }
 {
  bone1Name=base_shorts_l_02
  bone2Name=thigh_l
  swing1LimitDegrees=45
  swing1Motion=Free
  swing2LimitDegrees=45
  swing2Motion=Free
  twistLimitDegrees=45
  twistMotion=Free
 }
 {
  bone1Name=base_shorts_l_03
  bone2Name=thigh_l
  swing1LimitDegrees=45
  swing1Motion=Free
  swing2LimitDegrees=45
  swing2Motion=Free
  twistLimitDegrees=45
  twistMotion=Free
 }
 {
  bone1Name=base_shorts_l_04
  bone2Name=thigh_l
  swing1LimitDegrees=45
  swing1Motion=Free
  swing2LimitDegrees=45
  swing2Motion=Free
  twistLimitDegrees=45
  twistMotion=Free
 }
 {
  bone1Name=calf_twist_02_l
  bone2Name=thigh_l
  swing1LimitDegrees=45
  swing1Motion=Free
  swing2LimitDegrees=45
  swing2Motion=Free
  twistLimitDegrees=45
  twistMotion=Free
 }
]
